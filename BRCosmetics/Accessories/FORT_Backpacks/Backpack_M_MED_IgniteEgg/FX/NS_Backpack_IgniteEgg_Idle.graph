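# Idle Niagara effect for the IgniteEgg backpack cosmetic, drawing jittering, sketch-style black sprites whose visibility follows System.ReactivityNormalized.
# Front, Side and Top emitters burst lines at fixed per-index offsets, using VisibilityTag to sort particles.
# Jilililili spreads five sprites along a reactivity-scaled offset, lagging behind movement through interpolated world position.
# Hammer_Sketch_Lines spawns short-lived strokes at a reactivity-driven rate; some settings differ between front end and gameplay.
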NiagaraSystem /BRCosmetics/Accessories/FORT_Backpacks/Backpack_M_MED_IgniteEgg/FX/NS_Backpack_IgniteEgg_Idle
dependencies {
 usedDataInterfaces [&NiagaraDataInterfaceCurve]
 usedDynamicInputs [
  &MakeCustomFloatFromBool
  &Divide_Float
  &Subtract_Float
 ]
 usedModules [
  &SystemState
  &self:SystemSpawnScript.SystemScriptSource.SystemScriptGraph.NiagaraNodeAssignment_0.SetVariables_#1
 ]
}
emitters {
 Hammer_Sketch_Lines {
  emitterUpdateScript [
   &EmitterState(
    Inactive Response=::NewEnumerator1
    Life Cycle Mode=::NewEnumerator1
    Loop Behavior=::NewEnumerator1
    Loop Duration Mode=::NewEnumerator1
    UseLoopDelay=false
   )
   &SpawnRate(
    SpawnRate=&FloatFromCurve(
     FloatCurve=DataInterface() {propertyValues {Curve {
        defaultValue=0
        keys [
         {interpMode=RCIM_Cubic, tangentMode=RCTM_User, time=0.8, value=1}
         {arriveTangent=-12.5307, interpMode=RCIM_Cubic, leaveTangent=-12.5307, tangentMode=RCTM_User, time=1, value=0}
        ]
       }}}
     CurveIndex=$System.ReactivityNormalized
     Scale Curve=10
    )
    Use Spawn Probability=false
   ) {module=SpawnRate001}
  ]
  particleSpawnScript [
   &InitializeParticle(
    Color=<rgba 0 0 0 1>
    Lifetime Max=0.15
    Lifetime Min=0.2
    Lifetime Mode=::NewEnumerator1
    Position=&RandomRangeVector(Minimum=<0 -5 9>, Maximum=<0 -5 9.5>)
    Position Mode=::NewEnumerator1
    Sprite Rotation Angle=0.15
    Sprite Rotation Mode=::NewEnumerator1
    Sprite Size Max=<6 6>
    Sprite Size Min=<4 4>
    Sprite Size Mode=::NewEnumerator2
    UsePositionOffset=false
    Write Material Random=true
   )
   &CameraOffset(Camera Offset Amount=&FN_FrontEndFloatModifier(Gameplay=2, Front End=2))
  ]
  particleUpdateScript [
   &ParticleState()
   &ScaleColor(
    Scale Alpha=&FloatFromCurve(
     FloatCurve=DataInterface() {propertyValues {Curve {
        defaultValue=0
        keys [
         <at 0 0 RCIM_Cubic>
         {arriveTangent=0.0176, interpMode=RCIM_Cubic, leaveTangent=0.0176, tangentMode=RCTM_User, time=0.1, value=1}
         {interpMode=RCIM_Cubic, tangentMode=RCTM_User, time=0.9, value=1}
         {interpMode=RCIM_Cubic, tangentMode=RCTM_User, time=1, value=0}
        ]
       }}}
     CurveIndex=$System.ReactivityNormalized
     Scale Curve=2
    )
    ScaleA=true
    ScaleRGB=true
   )
   &OffsetPosition(Jitter Amount=0.5, Jitter Mode=1, Offset Mode=1)
   &SpriteFacingAndAlignment(
    Sprite Alignment=&MakeVector(X=&RandomRangeFloat(Minimum=-0.15, Maximum=0.15, Evaluation Type=::NewEnumerator1), Y=0, Z=1)
    Sprite Facing=<0 1 0>
    WriteAlignment=true
    WriteFacing=true
   ) {module=SpriteFacingAndAlignment001}
  ]
  rendererClasses [&NiagaraSpriteRendererProperties]
  renderers [{rendererClass=&NiagaraSpriteRendererProperties}]
 }
 Jilililili {
  emitterUpdateScript [
   &EmitterState(
    Inactive Response=::NewEnumerator1
    Life Cycle Mode=::NewEnumerator1
    Loop Behavior=::NewEnumerator1
    Loop Duration Mode=::NewEnumerator1
    UseLoopDelay=false
   )
   &SpawnBurst_Instantaneous(Spawn Count=5, Use Spawn Probability=false, UseLoopCountLimit=false)
  ]
  particleSpawnScript [
   &InitializeParticle(
    Color=<rgba 0 0 0 1>
    Lifetime=3.5
    Position Mode=::NewEnumerator0
    Sprite Rotation Angle=&RandomRangeFloat(Minimum=-0.05, Maximum=0.05)
    Sprite Rotation Mode=::NewEnumerator1
    Sprite Size Mode=::NewEnumerator3
    Write Material Random=true
   )
   /SetVariables(Particles.VisibilityTag=&MakeCustomIntFromBool(
     Bool=&SetBoolByIntComparison(A=&ReturnExecIndex(), B=3, Comparison Type=::NewEnumerator0)
     True INT=0
     False INT=1
    )) {isSetParameters=true}
  ]
  particleUpdateScript [
   &ParticleState(Kill Particles When Lifetime Has Elapsed=false)
   &ScaleSpriteSize(
    Scale Sprite Size Mode=::NewEnumerator0
    Uniform Scale Factor=&FloatFromCurve(
     FloatCurve=DataInterface() {propertyValues {Curve {
        defaultValue=1
        keys [
         {interpMode=RCIM_Cubic, tangentMode=RCTM_User, time=0, value=0.5}
         {arriveTangent=1.4344, interpMode=RCIM_Cubic, leaveTangent=1.4344, tangentMode=RCTM_User, time=1, value=1}
        ]
       }}}
     CurveIndex=&ReturnNormalizedExecIndex()
     Scale Curve=&FloatFromCurve(
      FloatCurve=DataInterface() {propertyValues {Curve {
         defaultValue=1
         keys [{interpMode=RCIM_Cubic, tangentMode=RCTM_User, time=0, value=0}, {interpMode=RCIM_Cubic, tangentMode=RCTM_User, time=0.1, value=1}]
        }}}
      CurveIndex=$System.ReactivityNormalized
     )
    )
   )
   &ScaleColor(
    Scale Alpha=&FloatFromCurve(
     FloatCurve=DataInterface() {propertyValues {Curve {
        defaultValue=0
        keys [
         <at 0 0 RCIM_Cubic>
         {arriveTangent=0.0176, interpMode=RCIM_Cubic, leaveTangent=0.0176, tangentMode=RCTM_User, time=0.1, value=1}
         {interpMode=RCIM_Cubic, tangentMode=RCTM_User, time=0.9, value=1}
         {interpMode=RCIM_Cubic, tangentMode=RCTM_User, time=1, value=0}
        ]
       }}}
     CurveIndex=$System.ReactivityNormalized
    )
    ScaleA=true
    ScaleRGB=true
   )
   &DynamicMaterialParameters(
    Index 0 Param 1=&FloatFromCurve(
     FloatCurve=DataInterface() {propertyValues {Curve {
        defaultValue=1
        keys [{interpMode=RCIM_Cubic, tangentMode=RCTM_User, time=0, value=0.5}, {interpMode=RCIM_Cubic, tangentMode=RCTM_User, time=1, value=1}]
       }}}
     CurveIndex=$System.ReactivityNormalized
    )
    Index 0 Param 2=1.5
    Index 0 Param 3=&RandomRangeFloat(Minimum=0, Maximum=1)
    Module.Index0Float0Write=true
    Module.Index0Float1Write=true
    Module.Index0Float2Write=true
    Module.Index0Float3Write=true
   )
   &InterpolateOverTime(
    Rate Of Change:NiagaraFloat=&FN_FrontEndFloatModifier(Gameplay=20, Front End=5)
    Target Vector Value:Vector3f=&TransformPosition(Position=/Niagara/DynamicInputs/Helpers/SimulationPosition(), Source Space=::Simulation, Destination Space=::World)
    Variable Type=::NewEnumerator2
   )
   /SetVariables(Particles.Position=&AddVectorToPosition(
     Position=&LerpPosition(
      A=<0 0 0>
      B=&TransformPosition(
       Position=Linked() {linkedVariable {name=Particles.InterpolateOverTime.InterpolateOverTime_V3_.Moving Average, type=Vector3f}}
       Source Space=::World
       Destination Space=::Simulation
      )
      Alpha=&FloatFromCurve(
       FloatCurve=DataInterface() {propertyValues {Curve {
          defaultValue=1
          keys [
           {interpMode=RCIM_Cubic, tangentMode=RCTM_User, time=0, value=0}
           {arriveTangent=2.3026, interpMode=RCIM_Cubic, leaveTangent=2.3026, tangentMode=RCTM_User, time=1, value=1}
          ]
         }}}
       CurveIndex=&ReturnNormalizedExecIndex()
      )
     )
     Vector=&Add_Vector(
      A=&MakeVector(
       X=&Add_Float(
        A=-10
        B=&FloatFromCurve(
         FloatCurve=DataInterface() {propertyValues {Curve {
            defaultValue=0
            keys [
             {arriveTangent=0.0418, interpMode=RCIM_Cubic, leaveTangent=0.0418, tangentMode=RCTM_User, time=0, value=0}
             {arriveTangent=1.4411, interpMode=RCIM_Cubic, leaveTangent=1.4411, tangentMode=RCTM_User, time=1, value=1}
            ]
           }}}
         CurveIndex=&ReturnNormalizedExecIndex()
         Scale Curve=&Multiply_Float(
          A=-15
          B=&FloatFromCurve(
           FloatCurve=DataInterface() {propertyValues {Curve {
              defaultValue=1
              keys [{arriveTangent=10.762, interpMode=RCIM_Cubic, leaveTangent=10.762, tangentMode=RCTM_User, time=0, value=0}, <at 0.2 1 RCIM_Cubic>]
             }}}
           CurveIndex=$System.ReactivityNormalized
          )
         )
        )
       )
       Y=&Add_Float(A=-5, B=&RandomRangeFloat(Minimum=-2, Maximum=2))
       Z=&Add_Float(
        A=13
        B=&FloatFromCurve(
         FloatCurve=DataInterface() {propertyValues {Curve {
            defaultValue=0
            keys [
             {arriveTangent=0.0418, interpMode=RCIM_Cubic, leaveTangent=0.0418, tangentMode=RCTM_User, time=0, value=0}
             {arriveTangent=1.0617, interpMode=RCIM_Cubic, leaveTangent=1.0617, tangentMode=RCTM_User, time=1, value=1}
            ]
           }}}
         CurveIndex=&ReturnNormalizedExecIndex()
         Scale Curve=&Multiply_Float(
          A=15
          B=&FloatFromCurve(
           FloatCurve=DataInterface() {propertyValues {Curve {
              defaultValue=1
              keys [{arriveTangent=13.2013, interpMode=RCIM_Cubic, leaveTangent=13.2013, tangentMode=RCTM_User, time=0, value=0}, <at 0.2 1 RCIM_Cubic>]
             }}}
           CurveIndex=$System.ReactivityNormalized
          )
         )
        )
       )
      )
      B=&RandomVector(Vector Scale=0.5)
     )
    )) {isSetParameters=true}
   &OffsetPosition(
    Jitter Amount:NiagaraFloat=&FloatFromCurve(
     FloatCurve=DataInterface() {propertyValues {Curve {
        defaultValue=0
        keys [
         {interpMode=RCIM_Cubic, tangentMode=RCTM_User, time=0.5, value=1}
         {arriveTangent=-1.7296, interpMode=RCIM_Cubic, leaveTangent=-1.7296, tangentMode=RCTM_User, time=1, value=0.675}
        ]
       }}}
     CurveIndex=$System.ReactivityNormalized
     Scale Curve=0.5
    )
    Jitter Delay:NiagaraFloat=&FloatFromCurve(
     FloatCurve=DataInterface() {propertyValues {Curve {
        defaultValue=1
        keys [
         {interpMode=RCIM_Cubic, tangentMode=RCTM_User, time=0.5, value=0.2}
         {arriveTangent=2.5091, interpMode=RCIM_Cubic, leaveTangent=2.5091, tangentMode=RCTM_User, time=1, value=0.6}
        ]
       }}}
     CurveIndex=$System.ReactivityNormalized
     Scale Curve=0.1
    )
    Jitter Interpolation=1
    Jitter Mode=1
    Jitter Time Offset=0.05
    Offset Mode=1
   ) {module=OffsetPosition002}
   /SetVariables(Particles.SpriteRotation:NiagaraFloat=&RandomRangeFloat(Minimum=-15, Maximum=15, Evaluation Type=::NewEnumerator1)) {isSetParameters=true}
  ]
  rendererClasses [&NiagaraSpriteRendererProperties]
  renderers [
   {rendererClass=&NiagaraSpriteRendererProperties}
   {rendererClass=&NiagaraSpriteRendererProperties, rendererIndex=1}
  ]
 }
 Sketch_Lines_Front {
  emitterUpdateScript [
   &EmitterState(
    Inactive Response=::NewEnumerator1
    Life Cycle Mode=::NewEnumerator1
    Loop Behavior=::NewEnumerator1
    Loop Duration Mode=::NewEnumerator1
    UseLoopDelay=false
   )
   &SpawnBurst_Instantaneous(Spawn Count=8, Spawn Time=0.1, Use Spawn Probability=false, UseLoopCountLimit=false)
  ]
  particleSpawnScript [
   &PartitionParticles(ID to Partition:NiagaraInt32=&ReturnExecIndex(), Number Of Partitions=8)
   &InitializeParticle(
    Color=<rgba 0 0 0 2>
    Lifetime=3.5
    Position Offset=&MergeHelper_VectorSwitch(
     0=<-8 -6 -10>
     1=<-11 -6 -3>
     2=<-12 -6 5>
     3=<-9 -6 15>
     4=<7.5 -6 -12>
     5=<11 -6 -2>
     6=<12 -6 5>
     7=<10 -6 15>
     Index=$Particles.PartitionParticles.Partition
     Num=::NewEnumerator6
    )
    Sprite Rotation Angle=&MergeHelper_FloatSwitch(0=0.05, 1=0.17, 2=0.05, 3=0.26, 4=0.78, 5=-0.4, 6=-0.255, 7=0.55, Index=$Particles.PartitionParticles.Partition, Num=::NewEnumerator6)
    Sprite Rotation Mode=::NewEnumerator1
    Sprite Size Mode=::NewEnumerator3
    Uniform Sprite Size=&MergeHelper_FloatSwitch(0=5, 1=3, 2=6, 3=7, 4=8, 5=6, 6=4, 7=6, Index=$Particles.PartitionParticles.Partition, Num=::NewEnumerator6)
    UsePositionOffset=true
    Write Material Random=true
   )
   /SetVariables(Particles.VisibilityTag=&MergeHelper_IntSwitch(0=1, 1=0, 2=0, 3=1, 4=1, 5=0, 6=0, 7=1, Index=$Particles.PartitionParticles.Partition, Num=::NewEnumerator6)) {isSetParameters=true, module=SetVariables_#2}
   &CameraOffset(Camera Offset Amount=5)
  ]
  particleUpdateScript [
   &ParticleState(Kill Particles When Lifetime Has Elapsed=false)
   &ScaleColor(
    Scale Alpha=&FloatFromCurve(
     FloatCurve=DataInterface() {propertyValues {Curve {
        defaultValue=0
        keys [
         <at 0 0 RCIM_Cubic>
         {arriveTangent=0.0176, interpMode=RCIM_Cubic, leaveTangent=0.0176, tangentMode=RCTM_User, time=0.1, value=1}
         {interpMode=RCIM_Cubic, tangentMode=RCTM_User, time=0.9, value=1}
         {interpMode=RCIM_Cubic, tangentMode=RCTM_User, time=1, value=0}
        ]
       }}}
     CurveIndex=$System.ReactivityNormalized
     Scale Curve=3
    )
    ScaleA=true
    ScaleRGB=true
   )
   &OffsetPosition(
    Jitter Amount:NiagaraFloat=&RandomRangeFloat(Minimum=0.8, Maximum=1.2)
    Jitter Delay=0.05
    Jitter Mode=1
    Offset Mode=1
   )
   &SpriteFacingAndAlignment(
    Sprite Alignment=&MakeVector(X=-1, Y=0, Z=&RandomRangeFloat(Minimum=-0.15, Maximum=0.15, Evaluation Type=::NewEnumerator1))
    Sprite Facing=<0 1 0>
    WriteAlignment=true
    WriteFacing=true
   ) {module=SpriteFacingAndAlignment001}
  ]
  rendererClasses [&NiagaraSpriteRendererProperties]
  renderers [
   {rendererClass=&NiagaraSpriteRendererProperties}
   {rendererClass=&NiagaraSpriteRendererProperties, rendererIndex=1}
  ]
 }
 Sketch_Lines_Side {
  emitterUpdateScript [
   &EmitterState(
    Inactive Response=::NewEnumerator1
    Life Cycle Mode=::NewEnumerator1
    Loop Behavior=::NewEnumerator1
    Loop Duration Mode=::NewEnumerator1
    UseLoopDelay=false
   )
   &SpawnBurst_Instantaneous(Spawn Count=6, Spawn Time=0.1, Use Spawn Probability=false, UseLoopCountLimit=false)
  ]
  particleSpawnScript [
   &PartitionParticles(ID to Partition:NiagaraInt32=&ReturnExecIndex(), Number Of Partitions=10)
   &InitializeParticle(
    Color=<rgba 0 0 0 2>
    Lifetime=3.5
    Position Offset=&MergeHelper_VectorSwitch(
     0=<0 -1 -11>
     1=<0 -12 -9>
     2=<0 2 -2>
     3=<0 -12 0>
     4=<0 2 9>
     5=<0 -10 11>
     Index=$Particles.PartitionParticles.Partition
     Num=::NewEnumerator4
    )
    Sprite Rotation Angle=&MergeHelper_FloatSwitch(0=0.3, 1=0.55, 2=0.15, 3=0.65, 4=0.051, 5=0.75, Index=$Particles.PartitionParticles.Partition, Num=::NewEnumerator4)
    Sprite Rotation Mode=::NewEnumerator1
    Sprite Size Mode=::NewEnumerator3
    Uniform Sprite Size=&MergeHelper_FloatSwitch(0=5, 1=6, 2=6, 3=5, 4=8, 5=5, Index=$Particles.PartitionParticles.Partition, Num=::NewEnumerator4)
    UsePositionOffset=true
    Write Material Random=true
   )
   /SetVariables(Particles.VisibilityTag=&MergeHelper_IntSwitch(0=1, 1=1, 2=0, 3=0, 4=1, 5=1, Index=$Particles.PartitionParticles.Partition, Num=::NewEnumerator4)) {isSetParameters=true, module=SetVariables_#2}
   &CameraOffset(Camera Offset Amount=5)
  ]
  particleUpdateScript [
   &ParticleState(Kill Particles When Lifetime Has Elapsed=false)
   &ScaleColor(
    Scale Alpha=&FloatFromCurve(
     FloatCurve=DataInterface() {propertyValues {Curve {
        defaultValue=0
        keys [
         <at 0 0 RCIM_Cubic>
         {arriveTangent=0.0176, interpMode=RCIM_Cubic, leaveTangent=0.0176, tangentMode=RCTM_User, time=0.1, value=1}
         {interpMode=RCIM_Cubic, tangentMode=RCTM_User, time=0.9, value=1}
         {interpMode=RCIM_Cubic, tangentMode=RCTM_User, time=1, value=0}
        ]
       }}}
     CurveIndex=$System.ReactivityNormalized
     Scale Curve=3
    )
    ScaleA=true
    ScaleRGB=true
   )
   &OffsetPosition(
    Jitter Amount:NiagaraFloat=&RandomRangeFloat(Minimum=0.8, Maximum=1.2)
    Jitter Delay=0.05
    Jitter Mode=1
    Offset Mode=1
   )
   &SpriteFacingAndAlignment(
    Sprite Alignment=&MakeVector(X=0, Y=1, Z=&RandomRangeFloat(Minimum=-0.2, Maximum=0.2, Evaluation Type=::NewEnumerator1))
    Sprite Facing=<-1 0 0>
    WriteAlignment=true
    WriteFacing=true
   )
  ]
  rendererClasses [&NiagaraSpriteRendererProperties]
  renderers [
   {rendererClass=&NiagaraSpriteRendererProperties}
   {rendererClass=&NiagaraSpriteRendererProperties, rendererIndex=1}
  ]
 }
 Sketch_Lines_Top {
  emitterUpdateScript [
   &EmitterState(
    Inactive Response=::NewEnumerator1
    Life Cycle Mode=::NewEnumerator1
    Loop Behavior=::NewEnumerator1
    Loop Duration Mode=::NewEnumerator1
    UseLoopDelay=false
   )
   &SpawnBurst_Instantaneous(Spawn Count=6, Spawn Time=0.1, Use Spawn Probability=false, UseLoopCountLimit=false)
  ]
  particleSpawnScript [
   &PartitionParticles(ID to Partition:NiagaraInt32=&ReturnExecIndex(), Number Of Partitions=10)
   &InitializeParticle(
    Color=<rgba 0 0 0 2>
    Lifetime=3.5
    Position Offset=&MergeHelper_VectorSwitch(
     0=<10 1 0>
     1=<10 -12 0>
     2=<0 2 0>
     3=<0 -13 0>
     4=<-11 1 0>
     5=<-10 -12 0>
     Index=$Particles.PartitionParticles.Partition
     Num=::NewEnumerator4
    )
    Sprite Rotation Angle=&MergeHelper_FloatSwitch(0=0.3, 1=0, 2=0.4, 3=-0.1, 4=0.5, 5=-0.2, Index=$Particles.PartitionParticles.Partition, Num=::NewEnumerator4)
    Sprite Rotation Mode=::NewEnumerator1
    Sprite Size Mode=::NewEnumerator3
    Uniform Sprite Size=&MergeHelper_FloatSwitch(0=5, 1=6, 2=7, 3=5, 4=8, 5=5, Index=$Particles.PartitionParticles.Partition, Num=::NewEnumerator4)
    UsePositionOffset=true
    Write Material Random=true
   )
   /SetVariables(Particles.VisibilityTag=&MergeHelper_IntSwitch(0=1, 1=1, 2=0, 3=0, 4=1, 5=1, Index=$Particles.PartitionParticles.Partition, Num=::NewEnumerator4)) {isSetParameters=true, module=SetVariables_#2}
   &CameraOffset(Camera Offset Amount=10)
  ]
  particleUpdateScript [
   &ParticleState(Kill Particles When Lifetime Has Elapsed=false)
   &ScaleColor(
    Scale Alpha=&FloatFromCurve(
     FloatCurve=DataInterface() {propertyValues {Curve {
        defaultValue=0
        keys [
         <at 0 0 RCIM_Cubic>
         {arriveTangent=0.0176, interpMode=RCIM_Cubic, leaveTangent=0.0176, tangentMode=RCTM_User, time=0.1, value=1}
         {interpMode=RCIM_Cubic, tangentMode=RCTM_User, time=0.9, value=1}
         {interpMode=RCIM_Cubic, tangentMode=RCTM_User, time=1, value=0}
        ]
       }}}
     CurveIndex=$System.ReactivityNormalized
     Scale Curve=3
    )
    ScaleA=true
    ScaleRGB=true
   )
   &OffsetPosition(Jitter Amount=1, Jitter Delay=0.05, Jitter Mode=1, Offset Mode=1)
   &SpriteFacingAndAlignment(
    Sprite Alignment=&MakeVector(X=1, Y=&RandomRangeFloat(Minimum=-0.2, Maximum=0.2, Evaluation Type=::NewEnumerator1), Z=0)
    Sprite Facing=<0 0 1>
    WriteAlignment=true
    WriteFacing=true
   )
  ]
  rendererClasses [&NiagaraSpriteRendererProperties]
  renderers [
   {rendererClass=&NiagaraSpriteRendererProperties}
   {rendererClass=&NiagaraSpriteRendererProperties, rendererIndex=1}
  ]
 }
}
summary {
 systemName=NS_Backpack_IgniteEgg_Idle
 emitters [
  {emitterName=Sketch_Lines_Side, rendererClasses [&NiagaraSpriteRendererProperties]}
  {emitterName=Sketch_Lines_Front, rendererClasses [&NiagaraSpriteRendererProperties]}
  {emitterName=Jilililili, rendererClasses [&NiagaraSpriteRendererProperties]}
  {emitterName=Sketch_Lines_Top, rendererClasses [&NiagaraSpriteRendererProperties]}
  {emitterName=Hammer_Sketch_Lines, rendererClasses [&NiagaraSpriteRendererProperties]}
 ]
 userVariables [
  User.bIsFrontEnd {type=NiagaraBool}
  User.bIsFrontEndPreview {type=NiagaraBool}
  User.DebugReactivity {type=NiagaraBool}
  User.ReactivityTime {type=NiagaraFloat, defaultValue=0, description=""}
 ]
}
