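# Animation Blueprint for the IgniteEgg alarm-clock backpack, which runs its clock hands and rings when the player scores an elimination.
# Additive bone transforms turn the hands from elapsed time since initialization, with the second hand ticking each second.
# A kill-stat watcher, or a looping frontend preview timer, briefly enables blended sine-wave shaking on hammer, bells and pack.
# The Stage1 variant check selects this procedural pose; otherwise the pose is copied from a mesh.
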AnimBlueprint /BRCosmetics/Accessories/FORT_Backpacks/Backpack_M_MED_IgniteEgg/Meshes/M_MED_IgniteEgg_Pack_AnimBP
target_skeleton=&/M_MED_IgniteEgg_Pack_Skeleton
graphs [
 AnimGraph {nodes [
   n1 = "Transform (Modify) Bone - Bone: dyn_secondHand"(ComponentPose=Local To Component(), Alpha=1, Translation="0,0,0", Rotation=Get secondHand_ROT() -> (secondHand_ROT:Rotator), Scale="1,1,1") {
    rotationMode=BMM_Additive
    rotationSpace=BCS_BoneSpace
    translationSpace=BCS_BoneSpace
    alphaBoolBlend {blendOption=Linear}
    boneToModify {boneName=dyn_secondHand}
   }
   n2 = "Transform (Modify) Bone - Bone: dyn_minuteHand"(ComponentPose=@n1, Alpha=1, Translation="0,0,0", Rotation=Get minuteHand_ROT() -> (minuteHand_ROT:Rotator), Scale="1,1,1") {
    rotationMode=BMM_Additive
    rotationSpace=BCS_BoneSpace
    translationSpace=BCS_BoneSpace
    alphaBoolBlend {blendOption=Linear}
    boneToModify {boneName=dyn_minuteHand}
   }
   n3 = "Transform (Modify) Bone - Bone: dyn_hourHand"(ComponentPose=@n2, Alpha=1, Translation="0,0,0", Rotation=Get hourHand_ROT() -> (hourHand_ROT:Rotator), Scale="1,1,1") {
    rotationMode=BMM_Additive
    rotationSpace=BCS_BoneSpace
    translationSpace=BCS_BoneSpace
    alphaBoolBlend {blendOption=Linear}
    boneToModify {boneName=dyn_hourHand}
   }
   Component To Local(ComponentPose=@n3)
   n4 = "Transform (Modify) Bone - Bone: dyn_secondHand"(ComponentPose=Local To Component(LocalPose=Input Pose()), Alpha=1, Translation="0,0,0", Rotation="0,0,191", Scale="1,1,1") {
    rotationMode=BMM_Additive
    rotationSpace=BCS_BoneSpace
    translationSpace=BCS_BoneSpace
    alphaBoolBlend {blendOption=Linear}
    boneToModify {boneName=dyn_secondHand}
    rotation {roll=191}
   }
   n5 = Get Active_Kill() -> (Active_Kill:bool)
   n6 = "Transform (Modify) Bone - Bone: dyn_hammer"!(ComponentPose=@n4, bAlphaBoolEnabled:bool=@n5, Translation="0,0,0", Rotation=Get hammerRing_ROT() -> (hammerRing_ROT:Rotator), Scale="1,1,1") -> (Pose) {
    alphaInputType=Bool
    rotationMode=BMM_Additive
    rotationSpace=BCS_BoneSpace
    translationSpace=BCS_BoneSpace
    alphaBoolBlend {blendInTime=0.3, blendOption=Linear, blendOutTime=0.8}
    boneToModify {boneName=dyn_hammer}
   }
   n7 = Get bellRing_ROT() -> (bellRing_ROT:Rotator)
   n8 = "Transform (Modify) Bone - Bone: dyn_bell_l"!(ComponentPose=@n6, bAlphaBoolEnabled:bool=@n5, Translation="0,0,0", Rotation=@n7, Scale="1,1,1") -> (Pose) {
    alphaInputType=Bool
    rotationMode=BMM_Additive
    rotationSpace=BCS_BoneSpace
    translationSpace=BCS_BoneSpace
    alphaBoolBlend {blendInTime=0.3, blendOption=Linear, blendOutTime=0.8}
    boneToModify {boneName=dyn_bell_l}
   }
   n9 = "Transform (Modify) Bone - Bone: dyn_bell_r"!(ComponentPose=@n8, bAlphaBoolEnabled:bool=@n5, Translation="0,0,0", Rotation=@n7, Scale="1,1,1") -> (Pose) {
    alphaInputType=Bool
    rotationMode=BMM_Additive
    rotationSpace=BCS_BoneSpace
    translationSpace=BCS_BoneSpace
    alphaBoolBlend {blendInTime=0.3, blendOption=Linear, blendOutTime=0.8}
    boneToModify {boneName=dyn_bell_r}
   }
   n10 = "Transform (Modify) Bone - Bone: pack"!(ComponentPose=@n9, bAlphaBoolEnabled:bool=@n5, Translation="0,0,0", Rotation=Get clockRing_ROT() -> (clockRing_ROT:Rotator), Scale) -> (Pose) {
    alphaInputType=Bool
    rotationMode=BMM_Additive
    rotationSpace=BCS_BoneSpace
    translationSpace=BCS_BoneSpace
    alphaBoolBlend {blendInTime=0.3, blendOption=Linear, blendOutTime=0.8}
    boneToModify {boneName=pack}
   }
   n11 = "Transform (Modify) Bone - Bone: dyn_secondHand"(ComponentPose=@n10, Alpha=1, Translation="0,0,0", Rotation=Get secondHand_ROT(), Scale="1,1,1") {
    rotationMode=BMM_Additive
    rotationSpace=BCS_BoneSpace
    translationSpace=BCS_BoneSpace
    alphaBoolBlend {blendOption=Linear}
    boneToModify {boneName=dyn_secondHand}
   }
   n12 = "Transform (Modify) Bone - Bone: dyn_minuteHand"(ComponentPose=@n11, Alpha=1, Translation="0,0,0", Rotation=Get minuteHand_ROT(), Scale="1,1,1") {
    rotationMode=BMM_Additive
    rotationSpace=BCS_BoneSpace
    translationSpace=BCS_BoneSpace
    alphaBoolBlend {blendOption=Linear}
    boneToModify {boneName=dyn_minuteHand}
   }
   n13 = "Transform (Modify) Bone - Bone: dyn_hourHand"(ComponentPose=@n12, Alpha=1, Translation="0,0,0", Rotation=Get hourHand_ROT(), Scale="1,1,1") {
    rotationMode=BMM_Additive
    rotationSpace=BCS_BoneSpace
    translationSpace=BCS_BoneSpace
    alphaBoolBlend {blendOption=Linear}
    boneToModify {boneName=dyn_hourHand}
   }
   Output Pose(Result=Blend Poses by bool(
     BlendPose_0=Component To Local(ComponentPose=@n13)
     BlendPose_1=Copy Pose From Mesh()
     bActiveValue=Get CosmeticVariantReactivity() -> (CosmeticVariantReactivity:bool)
    ) {blendType=Linear}) {layerGroup=None}
  ]}
 EventGraph {nodes [
   n1 = OnKill() {is=title K2Node_CustomEvent}
   > Set Active_Kill(Active_Kill:bool=true) -> (Output_Get:bool)
   > Delay(Duration=3)
   > Set Active_Kill(Active_Kill=false)
   Event BlueprintUpdateAnimation()
   > n2 = Sequence!(execute) -> (then_0, then_1, then_2:Exec, then_3:Exec, then_4:Exec)
   Branch(execute=@n2.then_2, Condition=Get bShouldDisplayFrontendReactivity() -> (bShouldDisplayFrontendReactivity:bool))
   > Do Once()
   > Delay(Duration=5)
   > OnKill() {is=title K2Node_CallFunction}
   > Set Timer by Event(Delegate=@n1, Time=8, bLooping=true)
   n3 = Cast To FortPlayerPawn(execute=@n2.then_3, Object=TryGetFortPawn())
   > n4 = WatchCosmeticStat!(
    execute:Exec
    Pawn:Fort Player Pawn Object Reference=@n3
    StatTag:Fort Stat Manager Tag Structure=(TagName="StatManager.AthenaKills")
    DefaultValue:int=0
    bBroadcastInitialValue:bool=false
    InitialValueBroadcastDelay:float=-1
    ChangeValueBroadcastDelay:float=-1
    bConsumeWaitingNotifications:bool=true
   ) -> (then:Exec, OnStatChanged:Exec, StatName:Name, StatValue:int) {is=title K2Node_AsyncAction}
   Set Elimination Count(execute=@n4.OnStatChanged, Elimination Count:double="To Float (Integer)"(InInt=@n4.StatValue)) -> (Output_Get:double)
   > OnKill()
   n5 = "Sin (Degrees)"!(A:double=float * float(A=GetGameTimeInSeconds(), B=6000)) -> (ReturnValue:double)
   Set hammerRing_ROT(execute=@n2.then_0, hammerRing_ROT:Rotator=MakeRotator(Pitch=float * float(A=@n5, B=25))) -> (Output_Get:Rotator)
   > Set bellRing_ROT(bellRing_ROT:Rotator=MakeRotator(Pitch=float * float(A=@n5, B=-6) {is=title K2Node_PromotableOperator})) -> (Output_Get:Rotator)
   > Set clockRing_ROT(clockRing_ROT:Rotator=MakeRotator(Pitch=float * float(A=@n5, B=1.5))) -> (Output_Get:Rotator)
   n6 = GetGameTimeInSeconds()
   n7 = Set TimeStart(execute=@n16.then, TimeStart:double=@n6) -> (Output_Get:double)
   n8 = float - float(A=@n6, B=@n7)
   n9 = float / float(A=@n8, B=-10)
   Set minuteHand_ROT(execute=@n2.then_1, minuteHand_ROT:Rotator=MakeRotator(Roll=NormalizeAxis!(Angle:float=@n9) -> (ReturnValue:float))) -> (Output_Get:Rotator)
   n10 = float / float(A=@n8, B=-120)
   > Set hourHand_ROT(hourHand_ROT:Rotator=MakeRotator(Roll=NormalizeAxis(Angle=@n10))) -> (Output_Get:Rotator)
   > Delay(Duration=1)
   n11 = float * float(A=@n8, B=-6)
   > Set secondHand_ROT(secondHand_ROT:Rotator=MakeRotator(Roll=NormalizeAxis(Angle=@n11))) -> (Output_Get:Rotator)
   Event OnCharacterCustomizationLoaded()
   > n12 = Cast To FortPlayerPawn(Object=TryGetFortPawn())
   > Set Pawn(Pawn:Fort Player Pawn Object Reference=@n12) -> (Output_Get:Fort Player Pawn Object Reference)
   > n13 = Cast To FortPlayerPawn(Object=GetOwningActor())
   > n14 = GetVariantChannelActiveIndexForItemActor!(
    execute
    Item:Athena Cosmetic Item Definition Object Reference=/BRCosmetics/Athena/Items/Cosmetics/Backpacks/Backpack_IgniteEgg
    Actor:Actor Object Reference=@n13
    Channel:Gameplay Tag Structure=(TagName="Cosmetics.Variant.Channel.Parts")
    VariantList:Array of Gameplay Tag Structures=Make Array("[0]"=(TagName="Cosmetics.Variant.Property.Stage1"))
    bErrorIfChannelMissing:bool=false
    bErrorIfTagNotInArray:bool=false
   ) -> (then, Index:int=0, bIndexIsValid:bool, bFoundChannel:bool)
   > n15 = Branch(Condition=@n14.bIndexIsValid)
   > Set CosmeticVariantReactivity(CosmeticVariantReactivity:bool=true) -> (Output_Get:bool)
   Set CosmeticVariantReactivity(execute=@n15.else, CosmeticVariantReactivity=false)
   n16 = Event BlueprintInitializeAnimation()
   TryGetPawnOwner()
  ]}
]
variables [
 Active_Kill {type=boolean}
 bellRing_ROT {type=Rotator}
 clockRing_ROT {type=Rotator}
 CosmeticVariantReactivity {type=boolean}
 Elimination Count {type=number}
 hammerRing_ROT {type=Rotator}
 hourHand_ROT {type=Rotator}
 minuteHand_ROT {type=Rotator}
 Pawn {type=FortPlayerPawn}
 secondHand_ROT {type=Rotator}
 TimeStart {type=number}
]
